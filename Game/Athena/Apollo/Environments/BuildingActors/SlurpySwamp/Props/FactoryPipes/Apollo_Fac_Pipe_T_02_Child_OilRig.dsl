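; Blueprint child of the Apollo factory T-pipe prop, an oil rig variant that can drip oil.
; On BeginPlay it calls the parent logic, then, if Is_Oil_Dripping is set, spawns the NS_OilDrip Niagara effect attached to its static mesh component.
; The overlap and tick events are empty, and the construction script only calls the parent.
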(event EventBeginPlay
  (|CallParentFunction)
  (if (Variables|Default|GetIs_Oil_Dripping)
    (Effects|SpawnFXSystemAttached "/Game/Effects/Fort_Effects/Effects/Environment/Apollo/NS_OilDrip.NS_OilDrip" (Variables|Debug|GetStaticMeshComponent))))

(event Collision|EventActorBeginOverlap
  (params (OtherActor :type Actor%20Object%20Reference)))

(event EventTick
  (params (DeltaSeconds :type Float%20%28single-precision%29)))

(fn ConstructionScript
  (|CallParentFunction))
